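# Niagara system for a short orange explosion light flash, stored with the rift build-up VFX.
# Its single Glow emitter burst-spawns particles drawn only by a light renderer, with no visible sprite.
# A curve fades particle alpha over lifetime, and the color with boosted alpha is written to Particles.LightColor.
# Sprite scale and dynamic material modules are disabled, and the User.BeatSpeed parameter is not referenced by the modules.
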NiagaraSystem /DragonCartNarrative/Assets/VFX/RiftBuildUp/NS_DCN_ExplosionLight
dependencies {usedDataInterfaces [&NiagaraDataInterfaceCurve], usedDynamicInputs [], usedModules [&SystemState]}
emitters {Glow {
  emitterUpdateScript [
   &EmitterState()
   &SpawnBurst_Instantaneous(Use Spawn Probability=false, UseLoopCountLimit=false)
  ]
  particleSpawnScript [
   &InitializeParticle(Color=<rgba 2 0.5747 0 1>, Lifetime=1.4, UsePositionOffset=false, Write Material Random=true)
   &CameraOffset(Camera Offset Amount=-10)
  ]
  particleUpdateScript [
   &ParticleState()
   &ScaleSpriteSize(Uniform Curve Sprite Scale=DataInterface() {propertyValues {Curve {keys [
        {arriveTangent=1.1961, interpMode=RCIM_Cubic, leaveTangent=1.1961, tangentMode=RCTM_User, time=0, value=0.2}
        {tangentMode=RCTM_User, time=1, value=1}
       ]}}}) {enabled=false}
   &ScaleColor(
    Scale Alpha=&FloatFromCurve(FloatCurve=DataInterface() {propertyValues {Curve {defaultValue=0, keys [{tangentMode=RCTM_User, time=0, value=1}, {tangentMode=RCTM_User, time=1, value=0}]}}})
    ScaleA=true
    ScaleRGB=true
   )
   /SetVariables(Particles.LightColor=&MultiplyLinearColor(Linear Color=$Particles.Color, Scale Factor=<1 1 1 15>)) {isSetParameters=true}
   &DynamicMaterialParameters(
    Index 0 Param 1=1
    Index 0 Param 2=&Lerp_Float(
     A=1.5
     B=0.2
     Alpha=&FloatFromCurve(FloatCurve=DataInterface() {propertyValues {Curve {keys [
          {arriveTangent=2.3663, interpMode=RCIM_Cubic, leaveTangent=2.3663, tangentMode=RCTM_User, time=0, value=0}
          {interpMode=RCIM_Cubic, tangentMode=RCTM_User, time=1, value=1}
         ]}}})
    )
    Index 3 Param 1=2
    Module.Index0Param0Write=true
    Module.Index0Param1Write=true
    Module.Index0Param2Write=false
    Module.Index0Param3Write=false
    Module.Index3Param0Write=true
    Module.Index3Param1Write=false
    Module.Index3Param2Write=false
    Module.Index3Param3Write=false
    Write Parameter Index 3=true
   ) {enabled=false}
  ]
  rendererClasses [&NiagaraLightRendererProperties]
  renderers [{rendererClass=&NiagaraLightRendererProperties, rendererIndex=0}]
 }}
summary {
 systemName=NS_DCN_ExplosionLight
 emitters [{emitterName=Glow, rendererClasses [&NiagaraLightRendererProperties]}]
 userVariables [User.BeatSpeed {type=NiagaraFloat, defaultValue=1, description=""}]
}
