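# Sound cue for energy arc lightning in the DragonCart rift sequence.
# A Random node picks evenly among four temporary monster-lightning wave players, then a Modulator varies the result.
# The cue plays quiet and pitched up through a dedicated energy-arcs attenuation and the Fort_Events sound class.
# A rift-lightning concurrency set is assigned, alongside override values limiting instances per owner.
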SoundCue /DragonCartNarrative/Assets/Audio/RiftSequence/SC_DragonCart_Seq_Rift_EnergyArcs
root=Modulator_0
nodes [
 {id=Modulator_0, children [Random_0]}
 {id=Random_0, children [WavePlayer_0,WavePlayer_2,WavePlayer_3,WavePlayer_5], properties {weights [1,1,1,1]}}
 {id=WavePlayer_0, properties {soundWaveAssetPtr=/DragonCartNarrative/Assets/Audio/RiftSequence/Wav/TEMP/LimeRock_100_Monster_Lightning_01}}
 {id=WavePlayer_2, properties {soundWaveAssetPtr=/DragonCartNarrative/Assets/Audio/RiftSequence/Wav/TEMP/LimeRock_100_Monster_Lightning_02}}
 {id=WavePlayer_3, properties {soundWaveAssetPtr=/DragonCartNarrative/Assets/Audio/RiftSequence/Wav/TEMP/LimeRock_100_Monster_Lightning_03}}
 {id=WavePlayer_5, properties {soundWaveAssetPtr=/DragonCartNarrative/Assets/Audio/RiftSequence/Wav/TEMP/LimeRock_100_Monster_Lightning_04}}
]
properties {
 attenuationSettings=&/ATT_DragonCart_Orb_Seq_EnergyArcs
 maxDistance=92356.3281
 pitchMultiplier=1.2
 soundClassObject=&Fort_Events
 volumeMultiplier=0.2
 concurrencyOverrides {bLimitToOwner=true, maxCount=2, retriggerTime=0.25}
 concurrencySet [&/CONC_DragonCart_Seq_RiftLightning]
}
