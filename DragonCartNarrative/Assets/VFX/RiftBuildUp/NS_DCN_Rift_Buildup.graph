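# Niagara system for the Dragon Cart rift build-up: orange electrical energy that intensifies as the rift charges.
# NE_Electricity_Wind uses spawn groups to combine arc sprites, inward-facing wind meshes and a long growing flare, ramping over User.BuildupDuration.
# NE_Zaps ramps up mesh bolts in a widening cylinder, each triggering arc audio at User.ZapAudioVolume.
# Both fade by proximity and time of day; ElectricArcs, Splines001 and Minimal are disabled.
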NiagaraSystem /DragonCartNarrative/Assets/VFX/RiftBuildUp/NS_DCN_Rift_Buildup
dependencies {
 usedDataInterfaces [&NiagaraDataInterfaceAudioPlayer, &NiagaraDataInterfaceCurve]
 usedDynamicInputs [&SimulationPosition, &UniformRangedFloat]
 usedModules [
  &SystemState
  &self:SystemSpawnScript.SystemScriptSource.SystemScriptGraph.NiagaraNodeAssignment_2.SetVariables_#1
  &self:SystemSpawnScript.SystemScriptSource.SystemScriptGraph.NiagaraNodeAssignment_3.SetVariables_#2
  &self:SystemSpawnScript.SystemScriptSource.SystemScriptGraph.NiagaraNodeAssignment_1.SetVariables_#3
  &self:SystemSpawnScript.SystemScriptSource.SystemScriptGraph.NiagaraNodeAssignment_0.SetVariables_#4
  &self:SystemSpawnScript.SystemScriptSource.SystemScriptGraph.NiagaraNodeAssignment_4.SetVariables_#5
 ]
}
emitters {
 ElectricArcs {
  bEnabled=false
  emitterUpdateScript [
   &EmitterState()
   &SpawnRate(
    Spawn Probability=0.7
    SpawnRate=&FloatFromCurve(
     FloatCurve=DataInterface() {propertyValues {Curve {
        defaultValue=0
        keys [
         {arriveTangent=3.8733, interpMode=RCIM_Cubic, leaveTangent=3.8733, tangentMode=RCTM_User, time=0, value=0}
         {arriveTangent=36.2369, interpMode=RCIM_Cubic, leaveTangent=36.2369, tangentMode=RCTM_User, time=1, value=8}
        ]
       }}}
     CurveIndex=$Emitter.NormalizedLoopAge
    )
    Use Spawn Probability=true
   )
  ]
  particleSpawnScript [
   &InitializeParticle(
    Color=<rgba 5 2.1198 0.2532 1>
    Lifetime=&FloatFromCurve(
     FloatCurve=DataInterface() {propertyValues {Curve {
        defaultValue=1
        keys [
         {interpMode=RCIM_Cubic, tangentMode=RCTM_User, time=0, value=1}
         {arriveTangent=-1.6186, interpMode=RCIM_Cubic, leaveTangent=-1.6186, tangentMode=RCTM_User, time=1, value=0.5}
        ]
       }}}
     CurveIndex=$Emitter.NormalizedLoopAge
     Scale Curve=&RandomRangeFloat(Minimum=0.5, Maximum=0.6)
    )
    Mass Mode=::NewEnumerator1
    Sprite Rotation Mode=::NewEnumerator2
    Sprite Size=&Multiply_Vector2D_ByFloat(
     Float=&FloatFromCurve(
      FloatCurve=DataInterface() {propertyValues {Curve {defaultValue=1, keys [<at 0 0.4 RCIM_Cubic>, {tangentMode=RCTM_User, time=1, value=1}]}}}
      CurveIndex=$Emitter.NormalizedLoopAge
     )
     Vector2D=&UniformRangedVector2D(Minimum=<200 200>, Maximum=<400 400>)
    )
    Sprite Size Mode=::NewEnumerator4
    UsePositionOffset=false
    Write Material Random=true
   ) {module=InitializeParticle001}
   &ShapeLocation(
    Cylinder Distribution=::NewEnumerator2
    Cylinder Radius=&Multiply_Float(
     A=1000
     B=&FloatFromCurve(
      FloatCurve=DataInterface() {propertyValues {Curve {
         defaultValue=1
         keys [
          {arriveTangent=4.1785, interpMode=RCIM_Cubic, leaveTangent=4.1785, tangentMode=RCTM_User, time=0, value=0}
          {arriveTangent=1.1179, interpMode=RCIM_Cubic, leaveTangent=1.1179, tangentMode=RCTM_User, time=0.2042, value=0.6228}
          {arriveTangent=0.0224, interpMode=RCIM_Cubic, leaveTangent=0.0224, tangentMode=RCTM_User, time=1, value=1}
         ]
        }}}
      CurveIndex=$Emitter.NormalizedLoopAge
     )
    )
    MaskSpawnGroup=false
    Offset=<0 0 -300>
    Offset Mode=::NewEnumerator0
    Shape Primitive=::NewEnumerator1
    U Position=&RandomRangeFloat(Minimum=0, Maximum=1)
    UseOwnerScale=true
   ) {module=ShapeLocation001}
   &DynamicMaterialParameters(
    Index 0 Param 1=$Particles.MaterialRandom
    Index 0 Param 2=&Lerp_Float(A=0.75, B=1.2, Alpha=$Particles.MaterialRandom)
    Index 0 Param 4=1
    Module.Param0WriteEnabled=true
    Module.Param1WriteEnabled=true
    Module.Param2WriteEnabled=false
    Module.Param3WriteEnabled=true
   )
   &CameraOffset(Camera Offset Amount=500)
   &ShapeLocation(MaskSpawnGroup=false, Override Seed=false, Sphere Surface Distribution=1, UseOwnerScale=true)
  ]
  particleUpdateScript [
   &ParticleState()
   &DynamicMaterialParameters(
    Index 0 Param 3=&FloatFromCurve(FloatCurve=DataInterface() {propertyValues {Curve {keys [<at 0.5 0 RCIM_Cubic>, {interpMode=RCIM_Cubic, leaveTangent=3.7979, time=1, value=1}]}}})
    Index 1 Param 2=&FloatFromCurve(FloatCurve=DataInterface() {propertyValues {Curve {keys [<at 0.25 1 RCIM_Cubic>,<at 0.5 0>]}}}, CurveIndex=$System.Age, Scale Curve=0.5)
    Index 1 Param 3=&FloatFromCurve(
     FloatCurve=DataInterface() {propertyValues {Curve {keys [
         {arriveTangent=4.106, interpMode=RCIM_Cubic, leaveTangent=4.106, tangentMode=RCTM_User, time=0.05, value=0}
         {tangentMode=RCTM_User, time=0.35, value=1}
        ]}}}
     CurveIndex=$System.Age
    )
    Index 1 Param 4=100
    Module.Param0WriteEnabled=false
    Module.Param0WriteEnabled1=true
    Module.Param1WriteEnabled=false
    Module.Param1WriteEnabled1=true
    Module.Param2WriteEnabled=true
    Module.Param2WriteEnabled1=true
    Module.Param3WriteEnabled=false
    Module.Param3WriteEnabled1=true
    Write Parameter Index 1=true
   ) {module=DynamicMaterialParameters001}
   &ScaleSpriteSize(
    Uniform Curve Scale=&Lerp_Float(A=0.5, B=1, Alpha=$User.SizeNormalized)
    Uniform Curve Sprite Scale=DataInterface() {propertyValues {Curve {keys [
        {arriveTangent=0.5666, interpMode=RCIM_Cubic, leaveTangent=0.5666, tangentMode=RCTM_User, time=0, value=0.75}
        {tangentMode=RCTM_User, time=1, value=1}
       ]}}}
   )
   &ScaleColor(
    Scale Alpha=&FloatFromCurve(
     FloatCurve=DataInterface() {propertyValues {Curve {
        defaultValue=1
        keys [
         {arriveTangent=2.7333, interpMode=RCIM_Cubic, leaveTangent=2.7333, tangentMode=RCTM_User, time=0, value=0}
         {interpMode=RCIM_Cubic, tangentMode=RCTM_User, time=1, value=1}
        ]
       }}}
     Scale Curve=0.5
    )
    Scale RGB=&VectorFromFloat(Value=&FloatFromCurve(FloatCurve=DataInterface() {propertyValues {Curve {
         defaultValue=1
         keys [
          {interpMode=RCIM_Cubic, tangentMode=RCTM_User, time=0, value=1}
          {interpMode=RCIM_Cubic, tangentMode=RCTM_User, time=0.5, value=5}
          {tangentMode=RCTM_User, time=1, value=1}
         ]
        }}}))
    ScaleA=true
    ScaleRGB=true
   ) {module=ScaleColor001}
   &ScaleColor(
    Scale Alpha=&FNTimeOfDayScalar(Night Scale=0.5, Morning Scale=0.8, Evening Scale=0.6)
    ScaleA=true
    ScaleRGB=true
   )
  ]
  rendererClasses [&NiagaraSpriteRendererProperties]
  renderers [{rendererClass=&NiagaraSpriteRendererProperties}]
 }
 Minimal {
  bEnabled=false
  emitterUpdateScript [
   &EmitterState()
   &SpawnBurst_Instantaneous(Use Spawn Probability=false, UseLoopCountLimit=false)
  ]
  particleSpawnScript [
   &InitializeParticle(
    Color=<rgba 5 5 5 1>
    Lifetime=5
    Mesh Scale=&Multiply_VectorByFloat(Vector=&RandomRangeVector(Minimum=<0.2 0.3 0.4>, Maximum=<0.3 0.6 0.6>), Float=35)
    Mesh Scale Mode=::NewEnumerator4
    UsePositionOffset=false
    Write Material Random=true
   )
   &InitialMeshOrientation(Mesh Orientation Mode=::NewEnumerator2, Use Rotation Vector=false)
  ]
  particleUpdateScript [
   &ParticleState()
   &DynamicMaterialParameters(
    Index 0 Param 1=&FloatFromCurve(FloatCurve=DataInterface() {propertyValues {Curve {defaultValue=0, keys [{tangentMode=RCTM_User, time=0, value=-1}, {tangentMode=RCTM_User, time=1, value=5}]}}})
    Index 0 Param 2=&Add_Float(
     A=&RandomRangeFloat(Minimum=0.4, Maximum=0.8)
     B=&FloatFromCurve(
      FloatCurve=DataInterface() {propertyValues {Curve {
         defaultValue=0
         keys [
          {arriveTangent=1.647, interpMode=RCIM_Cubic, leaveTangent=1.647, tangentMode=RCTM_User, time=0, value=0}
          {interpMode=RCIM_Cubic, tangentMode=RCTM_User, time=1, value=1}
         ]
        }}}
      CurveIndex=$Emitter.NormalizedLoopAge
      Scale Curve=&RandomRangeFloat(Minimum=0.5, Maximum=1.5)
     )
    )
    Index 0 Param 3=5000
    Module.Index0Float0Write=true
    Module.Index0Float1Write=true
    Module.Index0Float2Write=true
    Module.Index0Float3Write=true
   )
  ]
  rendererClasses [&NiagaraMeshRendererProperties]
  renderers [{rendererClass=&NiagaraMeshRendererProperties, rendererIndex=0}]
 }
 NE_Electricity_Wind {
  emitterUpdateScript [
   &EmitterState(Life Cycle Mode=::NewEnumerator1, Loop Behavior=::NewEnumerator1, Loop Duration=$User.BuildupDuration, UseLoopDelay=false)
   &SpawnBurst_Instantaneous(Spawn Time=28.8, Use Spawn Probability=false, UseLoopCountLimit=false) {module=SpawnBurst_Instantaneous001}
   &SpawnRate(
    Spawn Probability=0.7
    SpawnRate=&FloatFromCurve(
     FloatCurve=DataInterface() {propertyValues {Curve {
        defaultValue=0
        keys [
         {arriveTangent=0.4523, interpMode=RCIM_Cubic, leaveTangent=0.4523, tangentMode=RCTM_User, time=0, value=0}
         {arriveTangent=28.3592, interpMode=RCIM_Constant, tangentMode=RCTM_User, time=0.98, value=10}
         {arriveTangent=10.754, interpMode=RCIM_Constant, leaveTangent=10.754, time=0.99, value=0}
        ]
       }}}
     CurveIndex=$Emitter.NormalizedLoopAge
    )
    Use Spawn Probability=true
   )
   &SpawnRate(
    SpawnGroup=1
    SpawnRate=&FloatFromCurve(
     FloatCurve=DataInterface() {propertyValues {Curve {
        defaultValue=0
        keys [
         {
          arriveTangent=-0.6198
          arriveTangentWeight=0.67
          interpMode=RCIM_Cubic
          leaveTangent=-0.6198
          leaveTangentWeight=0.67
          tangentMode=RCTM_User
          tangentWeightMode=RCTWM_WeightedBoth
          time=0
          value=0
         }
         {arriveTangent=25.6785, interpMode=RCIM_Constant, tangentMode=RCTM_User, time=0.9638, value=10}
         {arriveTangent=10.754, interpMode=RCIM_Constant, leaveTangent=10.754, time=0.98, value=0}
        ]
       }}}
     CurveIndex=$Emitter.NormalizedLoopAge
    )
    Use Spawn Probability=true
   ) {module=SpawnRate001}
   &SpawnBurst_Instantaneous(Spawn Count=2, Spawn Group=2, Use Spawn Probability=false, UseLoopCountLimit=false)
  ]
  particleSpawnScript [
   /SetVariables(Particles.VisibilityTag=&SpawnGroup(Max Spawn Group=2)) {isSetParameters=true}
   &InitializeParticle(
    Color=&MergeHelper_ColorSwitch(
     0=<rgba 5 2.1198 0.2532 1>
     1=<rgba 20 5.8795 0 1>
     2=<rgba 5 2.1198 0.2532 1>
     3=?
     Index=&SpawnGroup()
     Num=::NewEnumerator1
    )
    Lifetime=&MergeHelper_FloatSwitch(
     0=&FloatFromCurve(
      FloatCurve=DataInterface() {propertyValues {Curve {
         defaultValue=1
         keys [
          {interpMode=RCIM_Cubic, tangentMode=RCTM_User, time=0, value=1}
          {arriveTangent=-0.8743, interpMode=RCIM_Cubic, leaveTangent=-0.8743, tangentMode=RCTM_User, time=1, value=0.7}
         ]
        }}}
      CurveIndex=$Emitter.NormalizedLoopAge
      Scale Curve=&RandomRangeFloat(Minimum=0.5, Maximum=0.6)
     )
     1=&FloatFromCurve(
      FloatCurve=DataInterface() {propertyValues {Curve {
         defaultValue=1
         keys [
          {interpMode=RCIM_Cubic, tangentMode=RCTM_User, time=0, value=1}
          {
           arriveTangent=-5.3613
           arriveTangentWeight=0.4969
           interpMode=RCIM_Cubic
           leaveTangent=-5.3613
           leaveTangentWeight=0.4969
           tangentMode=RCTM_User
           tangentWeightMode=RCTWM_WeightedBoth
           time=1
           value=0.5
          }
         ]
        }}}
      CurveIndex=$Emitter.NormalizedLoopAge
      Scale Curve=2
     )
     2=31
     3=?
     Index=&SpawnGroup()
     Num=::NewEnumerator1
    )
    Mesh Scale=&Multiply_VectorByFloat(
     Vector=&RandomRangeVector(Minimum=<0.2 0.3 0.4>, Maximum=<0.3 0.6 0.6>)
     Float=&FloatFromCurve(FloatCurve=DataInterface() {propertyValues {Curve {keys [<at 0 30>,<at 1 33>]}}}, CurveIndex=$Emitter.NormalizedLoopAge)
    )
    Mesh Scale Mode=::NewEnumerator4
    Sprite Rotation Mode=::NewEnumerator2
    Sprite Size=&MergeHelper_Vector2Switch(
     0=&Vector2DFromFloat(Value=&MakeFloatFromVector2d(Vector 2D=&Multiply_Vector2D_ByFloat(
        Float=&FloatFromCurve(
         FloatCurve=DataInterface() {propertyValues {Curve {defaultValue=1, keys [<at 0 0.4 RCIM_Cubic>, {tangentMode=RCTM_User, time=1, value=1}]}}}
         CurveIndex=$Emitter.NormalizedLoopAge
        )
        Vector2D=&UniformRangedVector2D(Minimum=<200 200>, Maximum=<400 400>)
       )))
     1=<0 0>
     2=<2000 1000>
     3=?
     4=?
     Index=&SpawnGroup()
     Num=::NewEnumerator1
    )
    Sprite Size Mode=::NewEnumerator4
    UsePositionOffset=false
    Write Material Random=true
   )
   &CameraOffset(Camera Offset Amount=&MergeHelper_FloatSwitch(0=300, 1=40, 2=1500, 3=?, Index=&SpawnGroup(), Num=::NewEnumerator1))
   &ShapeLocation(
    Cylinder Radius=&MergeHelper_FloatSwitch(
     0=&Multiply_Float(
      A=1000
      B=&FloatFromCurve(
       FloatCurve=DataInterface() {propertyValues {Curve {
          defaultValue=1
          keys [
           {arriveTangent=4.1785, interpMode=RCIM_Cubic, leaveTangent=4.1785, tangentMode=RCTM_User, time=0, value=0}
           {arriveTangent=1.1179, interpMode=RCIM_Cubic, leaveTangent=1.1179, tangentMode=RCTM_User, time=0.2042, value=0.6228}
           {arriveTangent=0.0224, interpMode=RCIM_Cubic, leaveTangent=0.0224, tangentMode=RCTM_User, time=1, value=1}
          ]
         }}}
       CurveIndex=$Emitter.NormalizedLoopAge
       Scale Curve=$NPC.NPC_DragonCart_Orb_Rift.RiftScale
      )
     )
     1=&Multiply_Float(A=1000, B=$NPC.NPC_DragonCart_Orb_Rift.RiftScale)
     2=0
     3=?
     Index=&SpawnGroup()
     Num=::NewEnumerator1
    )
    MaskSpawnGroup=false
    Override Seed=false
    Shape Primitive=::NewEnumerator1
    Surface Only=false
    UseOwnerScale=true
   )
   &DynamicMaterialParameters(
    Index 0 Param 1=$Particles.MaterialRandom
    Index 0 Param 2=&Lerp_Float(A=0.75, B=1.2, Alpha=$Particles.MaterialRandom)
    Index 0 Param 4=1
    Module.Param0WriteEnabled=true
    Module.Param1WriteEnabled=true
    Module.Param2WriteEnabled=false
    Module.Param3WriteEnabled=true
   )
   /SetVariables(Particles.MeshOrientation=&self:MakeQuatFromXandZ_02(
     XAxis:Vector3f=&Multiply_VectorByFloat(Vector=$Particles.ShapeLocation.ShapeNormal, Float=-1)
     ZAxis:Vector3f=$Particles.ShapeLocation.ShapeBitangent
    )) {isSetParameters=true}
  ]
  particleUpdateScript [
   &ParticleState()
   &ScaleSpriteSize(
    Scale Sprite Size Mode=::NewEnumerator0
    Uniform Scale Factor=&MergeHelper_FloatSwitch(
     0=&FloatFromCurve(
      FloatCurve=DataInterface() {propertyValues {Curve {keys [
          {arriveTangent=0.5666, interpMode=RCIM_Cubic, leaveTangent=0.5666, tangentMode=RCTM_User, time=0, value=0.75}
          {tangentMode=RCTM_User, time=1, value=1}
         ]}}}
      Scale Curve=&Lerp_Float(A=0.5, B=1, Alpha=$User.SizeNormalized)
     )
     1=0
     2=&FloatFromCurve(
      FloatCurve=DataInterface() {propertyValues {Curve {
         defaultValue=0
         keys [
          {interpMode=RCIM_Cubic, leaveTangentWeight=0.1326, tangentMode=RCTM_User, time=0, value=0}
          {arriveTangentWeight=0.1895, interpMode=RCIM_Cubic, leaveTangentWeight=0.2772, tangentMode=RCTM_User, time=0.2219, value=-0.0108}
          {
           arriveTangent=0.1358
           arriveTangentWeight=0.3025
           interpMode=RCIM_Cubic
           leaveTangentWeight=0.0167
           tangentMode=RCTM_Break
           tangentWeightMode=RCTWM_WeightedBoth
           time=1
           value=1
          }
         ]
        }}}
      Scale Curve=5
     )
     3=?
     Index=&SpawnGroup()
     Num=::NewEnumerator1
    )
   )
   &DynamicMaterialParameters(
    Index 0 Param 3=&FloatFromCurve(FloatCurve=DataInterface() {propertyValues {Curve {keys [<at 0.5 0 RCIM_Cubic>, {interpMode=RCIM_Cubic, leaveTangent=3.7979, time=1, value=1}]}}})
    Index 1 Param 2=&FloatFromCurve(FloatCurve=DataInterface() {propertyValues {Curve {keys [<at 0.25 1 RCIM_Cubic>,<at 0.5 0>]}}}, CurveIndex=$System.Age, Scale Curve=0.5)
    Index 1 Param 3=&FloatFromCurve(
     FloatCurve=DataInterface() {propertyValues {Curve {keys [
         {arriveTangent=4.106, interpMode=RCIM_Cubic, leaveTangent=4.106, tangentMode=RCTM_User, time=0.05, value=0}
         {tangentMode=RCTM_User, time=0.35, value=1}
        ]}}}
     CurveIndex=$System.Age
    )
    Index 1 Param 4=100
    Module.Param0WriteEnabled=false
    Module.Param0WriteEnabled1=true
    Module.Param1WriteEnabled=false
    Module.Param1WriteEnabled1=true
    Module.Param2WriteEnabled=true
    Module.Param2WriteEnabled1=true
    Module.Param3WriteEnabled=false
    Module.Param3WriteEnabled1=true
    Write Parameter Index 1=true
   ) {module=DynamicMaterialParameters001}
   &ScaleMeshSize(Scale Factor=&VectorFromFloat(Value=2))
   &ScaleColor(
    Scale Alpha=&MergeHelper_FloatSwitch(
     0=&FloatFromCurve(
      FloatCurve=DataInterface() {propertyValues {Curve {
         defaultValue=1
         keys [
          {arriveTangent=2.7333, interpMode=RCIM_Cubic, leaveTangent=2.7333, tangentMode=RCTM_User, time=0, value=0}
          {interpMode=RCIM_Cubic, tangentMode=RCTM_User, time=1, value=1}
         ]
        }}}
      Scale Curve=0.5
     )
     1=1
     2=&FloatFromCurve(FloatCurve=DataInterface() {propertyValues {Curve {
         defaultValue=0
         keys [
          {interpMode=RCIM_Cubic, tangentMode=RCTM_User, time=0, value=0}
          {arriveTangent=0.0652, interpMode=RCIM_Cubic, leaveTangent=0.0652, tangentMode=RCTM_User, time=0.4546, value=0}
          {
           arriveTangent=0.55
           arriveTangentWeight=0.1425
           interpMode=RCIM_Cubic
           leaveTangentWeight=0.0167
           tangentMode=RCTM_Break
           tangentWeightMode=RCTWM_WeightedBoth
           time=0.6966
           value=1.4
          }
          {arriveTangent=-4.6143, interpMode=RCIM_Cubic, leaveTangent=-4.6143, time=0.95, value=1.4}
          {arriveTangentWeight=1.124, interpMode=RCIM_Cubic, leaveTangent=9.7397, leaveTangentWeight=1.124, tangentMode=RCTM_User, time=1, value=0}
         ]
        }}})
     3=?
     Index=&SpawnGroup()
     Num=::NewEnumerator1
    )
    Scale RGB=&MergeHelper_VectorSwitch(
     0=&VectorFromFloat(Value=&FloatFromCurve(FloatCurve=DataInterface() {propertyValues {Curve {
          defaultValue=1
          keys [
           {interpMode=RCIM_Cubic, tangentMode=RCTM_User, time=0, value=1}
           {interpMode=RCIM_Cubic, tangentMode=RCTM_User, time=0.5, value=5}
           {tangentMode=RCTM_User, time=1, value=1}
          ]
         }}}))
     1=<1 1 1>
     2=&VectorFromFloat(Value=&FloatFromCurve(
       FloatCurve=DataInterface() {propertyValues {Curve {
          defaultValue=1
          keys [
           {interpMode=RCIM_Cubic, tangentMode=RCTM_User, time=0, value=1}
           {arriveTangent=3.243, interpMode=RCIM_Cubic, leaveTangent=3.243, tangentMode=RCTM_User, time=1, value=2}
          ]
         }}}
       Scale Curve=2
      ))
     3=?
     Index=&SpawnGroup()
     Num=::NewEnumerator1
    )
    ScaleA=true
    ScaleRGB=true
   )
   &ScaleColor(
    Scale Alpha=&Multiply_Float(A=$System.NearFade, B=&FNTimeOfDayScalar(Night Scale=0.5, Morning Scale=0.8, Evening Scale=0.6))
    ScaleA=true
    ScaleRGB=true
   ) {module=ScaleColor001}
   /SetVariables(
    Particles.FlareDyn:Vector4f=&MakeVector4(
     X=&FloatFromCurve(FloatCurve=DataInterface() {propertyValues {Curve {
         defaultValue=0
         keys [
          {arriveTangent=-2.0436, interpMode=RCIM_Cubic, leaveTangent=-2.0436, tangentMode=RCTM_User, time=0, value=1}
          {arriveTangent=0.1814, leaveTangent=0.1814, time=1, value=0.05}
         ]
        }}})
     Y=2
     Z=&FloatFromCurve(FloatCurve=DataInterface() {propertyValues {Curve {
         defaultValue=1
         keys [
          {tangentMode=RCTM_User, time=0, value=1}
          {arriveTangent=-0.0405, interpMode=RCIM_Cubic, leaveTangent=-0.0405, tangentMode=RCTM_User, time=0.7, value=1}
          {arriveTangent=-4.8751, interpMode=RCIM_Cubic, leaveTangent=-4.8751, tangentMode=RCTM_User, time=1, value=0}
         ]
        }}})
     W=0
    )
    Particles.WindDyn:Vector4f=&MakeVector4(
     X=&FloatFromCurve(FloatCurve=DataInterface() {propertyValues {Curve {defaultValue=1, keys [{tangentMode=RCTM_User, time=0, value=-1}, {tangentMode=RCTM_User, time=1, value=1}]}}})
     Y=&RandomRangeFloat(Minimum=0, Maximum=1)
     Z=&Multiply_Float(A=100, B=&RandomRangeFloat(Minimum=15, Maximum=18))
     W=0
    )
   ) {isSetParameters=true}
  ]
  rendererClasses [&NiagaraSpriteRendererProperties, &NiagaraMeshRendererProperties]
  renderers [
   {rendererClass=&NiagaraSpriteRendererProperties}
   {rendererClass=&NiagaraMeshRendererProperties, rendererIndex=1}
   {rendererClass=&NiagaraSpriteRendererProperties, rendererIndex=2}
  ]
 }
 NE_Zaps {
  emitterSpawnScript [&InitializeAudioPlayer(Audio Player=DataInterface() {propertyValues {
      Attenuation=null
      Concurrency=null
      MaxPlaysPerTick=10
      SoundToPlay=/DragonCartNarrative/Assets/Audio/RiftSequence/SC_DragonCart_Seq_Rift_EnergyArcs
      bAllowLoopingOneShotSounds=false
      bLimitPlaysPerTick=true
      bOnlyActiveDuringGameplay=false
      bStopWhenComponentIsDestroyed=true
      ConfigurationUserParameter {parameter {name=None, typeDefHandle {registeredTypeIndex=1}, varData []}}
      ParameterNames []
     }})]
  emitterUpdateScript [
   &EmitterState()
   &SpawnRate(
    SpawnRate=&FloatFromCurve(
     FloatCurve=DataInterface() {propertyValues {Curve {
        defaultValue=0
        keys [
         {arriveTangent=7.3529, interpMode=RCIM_Cubic, leaveTangent=7.3529, tangentMode=RCTM_User, time=0, value=0}
         {interpMode=RCIM_Cubic, tangentMode=RCTM_User, time=1, value=30}
        ]
       }}}
     CurveIndex=$Emitter.NormalizedLoopAge
    )
    Use Spawn Probability=false
   )
  ]
  particleSpawnScript [
   &InitializeParticle(
    Color=<rgba 5 1.9477 0 1>
    Lifetime=&FloatFromCurve(
     FloatCurve=DataInterface() {propertyValues {Curve {
        defaultValue=1
        keys [
         {interpMode=RCIM_Cubic, tangentMode=RCTM_User, time=0, value=1}
         {arriveTangent=-1.6186, interpMode=RCIM_Cubic, leaveTangent=-1.6186, tangentMode=RCTM_User, time=1, value=0.5}
        ]
       }}}
     CurveIndex=$Emitter.NormalizedLoopAge
     Scale Curve=&RandomRangeFloat(Minimum=0.5, Maximum=0.6)
    )
    Mesh Scale Mode=::NewEnumerator3
    Mesh Uniform Scale=&RandomRangeFloat(Minimum=0.3, Maximum=0.5)
    Sprite Size Mode=::NewEnumerator3
    Uniform Sprite Size=20
    UsePositionOffset=false
    Write Material Random=true
   )
   &ShapeLocation(
    Cylinder Distribution=::NewEnumerator2
    Cylinder Height=1000
    Cylinder Radius=&Multiply_Float(
     A=2200
     B=&FloatFromCurve(
      FloatCurve=DataInterface() {propertyValues {Curve {
         defaultValue=1
         keys [
          {arriveTangent=4.1785, interpMode=RCIM_Cubic, leaveTangent=4.1785, tangentMode=RCTM_User, time=0, value=0}
          {arriveTangent=1.1179, interpMode=RCIM_Cubic, leaveTangent=1.1179, tangentMode=RCTM_User, time=0.2042, value=0.6228}
          {arriveTangent=0.0224, interpMode=RCIM_Cubic, leaveTangent=0.0224, tangentMode=RCTM_User, time=1, value=1}
         ]
        }}}
      CurveIndex=$Emitter.NormalizedLoopAge
     )
    )
    MaskSpawnGroup=false
    Offset Mode=::NewEnumerator0
    Shape Primitive=::NewEnumerator1
    U Position=&RandomRangeFloat(Minimum=0, Maximum=1)
    UseOwnerScale=true
   )
   &InitialMeshOrientation(
    Mesh Orientation Mode=::NewEnumerator2
    Orientation Vector=&MaskVectorBySpawnGroup(Masked Vector=&RandomVector(Vector Scale=1), Passthrough Vector=<1 0 0>)
    Rotation=&MaskVectorBySpawnGroup(Masked Vector=<0 0 0>, Passthrough Vector=&RandomRangeVector(Minimum=<0 0 0>, Maximum=<0 0 1>))
    Use Rotation Vector=true
   )
   /Niagara/Modules/Audio/PlayAudio(Volume:NiagaraFloat=$User.ZapAudioVolume)
  ]
  particleUpdateScript [
   &ParticleState()
   &ScaleMeshSize(Scale Factor=&Multiply_VectorByFloat(
     Vector=&RandomRangeVector(Minimum=<1 1 2.2>, Maximum=<1.3 1.3 2.2>)
     Float=&FloatFromCurve(
      FloatCurve=DataInterface() {propertyValues {Curve {keys [<at 0 1>,<at 1 1.5>]}}}
      Scale Curve=&FloatFromCurve(
       FloatCurve=DataInterface() {propertyValues {Curve {keys [
           {arriveTangent=4.0995, interpMode=RCIM_Cubic, leaveTangent=4.0995, tangentMode=RCTM_User, time=0, value=4}
           {arriveTangent=5.9215, interpMode=RCIM_Cubic, leaveTangent=5.9215, tangentMode=RCTM_User, time=1, value=8}
          ]}}}
       CurveIndex=$Emitter.NormalizedLoopAge
      )
     )
    )) {module=ScaleMeshSize001}
   &DynamicMaterialParameters(
    Index 0 Param 2=&RandomRangeFloat(Minimum=0, Maximum=0)
    Index 0 Param 4=&MaskFloatBySpawnGroup(Masked Float=13, Passthrough Float=1)
    Index 1 Param 1=&FloatFromCurve(FloatCurve=DataInterface() {propertyValues {Curve {keys [<at 0 0.5>,<at 1 1>]}}})
    Index 1 Param 2=2
    Index 1 Param 3=&FloatFromCurve(FloatCurve=DataInterface() {propertyValues {Curve {keys [{interpMode=RCIM_Cubic, tangentMode=RCTM_Break, time=0.5, value=-0.5}, <at 1 0.1>]}}})
    Index 2 Param 1=&Multiply_Float(
     A=&FloatFromCurve(FloatCurve=DataInterface() {propertyValues {Curve {keys [{interpMode=RCIM_Cubic, leaveTangent=8.1085, tangentMode=RCTM_Break, time=0.8568, value=1}, {tangentMode=RCTM_User, time=1, value=0}]}}})
     B=&Divide_Float(
      A=&MakeFloatFromVector(Vector=$Particles.Scale, Channel=::NewEnumerator0)
      B=&RandomRangeFloat(Minimum=10, Maximum=15)
     )
    )
    Index 2 Param 2=&FloatFromCurve(
     FloatCurve=DataInterface() {propertyValues {Curve {keys [
         {interpMode=RCIM_Cubic, tangentMode=RCTM_Break, time=0.7, value=0}
         {arriveTangent=-0.0859, interpMode=RCIM_Cubic, tangentMode=RCTM_Break, time=1, value=-1}
        ]}}}
     Scale Curve=50
    )
    Index 2 Param 3=&Waveform(
     Global Input Position=&RandomRangeFloat(Minimum=0, Maximum=15)
     Global Phase=$Particles.Age
     "Waveform [1]"=::NewEnumerator12
     "[1] Phase Scale"=2
     "[1] Frequency"=1
     "[1] Amplitude Min/Max"=<0 1>
    )
    Index 2 Param 4=25
    Index 3 Param 1=&FloatFromCurve(
     FloatCurve=DataInterface() {propertyValues {Curve {keys [
         {interpMode=RCIM_Cubic, leaveTangent=-0.0025, tangentMode=RCTM_Break, time=0, value=0}
         {arriveTangent=1.8528, interpMode=RCIM_Cubic, tangentMode=RCTM_Break, time=1, value=1}
        ]}}}
     Scale Curve=&RandomRangeFloat(Minimum=0, Maximum=0)
    )
    Index 3 Param 2=&FloatFromCurve(
     FloatCurve=DataInterface() {propertyValues {Curve {keys [
         {interpMode=RCIM_Cubic, leaveTangent=-0.0025, tangentMode=RCTM_Break, time=0, value=0.5}
         {arriveTangent=-0.0316, interpMode=RCIM_Cubic, tangentMode=RCTM_Break, time=1, value=-0.2}
        ]}}}
     Scale Curve=0.1
    )
    Index 3 Param 3=&FloatFromCurve(FloatCurve=DataInterface() {propertyValues {Curve {keys [
         {interpMode=RCIM_Cubic, leaveTangent=1.1034, tangentMode=RCTM_Break, time=0, value=-1}
         {arriveTangent=2.2399, tangentMode=RCTM_User, time=0.7, value=0.125}
        ]}}})
    Index 3 Param 4=&FloatFromCurve(FloatCurve=DataInterface() {propertyValues {Curve {keys [
         {interpMode=RCIM_Cubic, leaveTangent=3.0601, tangentMode=RCTM_Break, time=0.5, value=0}
         {arriveTangent=3.2053, tangentMode=RCTM_User, time=1, value=1}
        ]}}})
    Module.Index0Param0Write=true
    Module.Index0Param1Write=true
    Module.Index0Param2Write=true
    Module.Index0Param3Write=true
    Module.Index1Param0Write=true
    Module.Index1Param1Write=true
    Module.Index1Param2Write=true
    Module.Index1Param3Write=true
    Module.Index2Param0Write=true
    Module.Index2Param1Write=true
    Module.Index2Param2Write=true
    Module.Index2Param3Write=true
    Module.Index3Param0Write=true
    Module.Index3Param1Write=true
    Module.Index3Param2Write=true
    Module.Index3Param3Write=true
    Write Parameter Index 1=true
    Write Parameter Index 2=true
    Write Parameter Index 3=true
   )
   &SolveForcesAndVelocity(Clamp Velocity=false, Limit Acceleration=false)
   /SetVariables(Particles.DecalColor:LinearColor=&MultiplyLinearColor(
     Linear Color=<rgba 1 0.2038 0.0156 1>
     Scale Factor=&MakeVector4_FromLinearColor(LinearColor=$Particles.Color)
    )) {isSetParameters=true}
   /SetVariables(Particles.GlowDecalSize:Vector2f=&Multiply_Vector2D_ByFloat(Float=60, Vector2D=<1 1>)) {isSetParameters=true}
   /SetVariables(Particles.DecalDynMat:Vector4f=<3 1 0 0>) {isSetParameters=true}
   &CameraOffset(Camera Offset Amount=100)
   &ScaleColor(
    Scale Alpha=&Multiply_Float(A=$System.NearFade, B=&FNTimeOfDayScalar(Night Scale=0.5, Morning Scale=0.8, Evening Scale=0.6))
    ScaleA=true
    ScaleRGB=true
   )
  ]
  rendererClasses [&NiagaraMeshRendererProperties]
  renderers [
   {rendererClass=&NiagaraMeshRendererProperties, rendererIndex=0}
   {rendererClass=&NiagaraMeshRendererProperties, rendererIndex=1}
  ]
 }
 Splines001 {
  bEnabled=false
  emitterUpdateScript [
   &EmitterState()
   &SpawnRate(
    SpawnRate=&FloatFromCurve(
     FloatCurve=DataInterface() {propertyValues {Curve {
        defaultValue=0
        keys [
         {
          arriveTangent=1.3627
          arriveTangentWeight=0.4603
          interpMode=RCIM_Cubic
          leaveTangent=1.3627
          leaveTangentWeight=0.4603
          tangentMode=RCTM_User
          tangentWeightMode=RCTWM_WeightedBoth
          time=0
          value=0
         }
         {arriveTangent=37.893, interpMode=RCIM_Cubic, leaveTangent=37.893, tangentMode=RCTM_User, time=1, value=10}
        ]
       }}}
     CurveIndex=$Emitter.NormalizedLoopAge
    )
    Use Spawn Probability=true
   )
  ]
  particleSpawnScript [
   &InitializeParticle(
    Color=<rgba 20 5.8795 0 1>
    Lifetime=&FloatFromCurve(
     FloatCurve=DataInterface() {propertyValues {Curve {
        defaultValue=1
        keys [
         {interpMode=RCIM_Cubic, tangentMode=RCTM_User, time=0, value=1}
         {arriveTangent=-1.6186, interpMode=RCIM_Cubic, leaveTangent=-1.6186, tangentMode=RCTM_User, time=1, value=0.5}
        ]
       }}}
     CurveIndex=$Emitter.NormalizedLoopAge
     Scale Curve=&RandomRangeFloat(Minimum=1.5, Maximum=2)
    )
    Mesh Scale=&Multiply_VectorByFloat(
     Vector=&RandomRangeVector(Minimum=<0.2 0.3 0.4>, Maximum=<0.3 0.6 0.6>)
     Float=&FloatFromCurve(FloatCurve=DataInterface() {propertyValues {Curve {keys [<at 0 30>,<at 1 40>]}}}, CurveIndex=$Emitter.NormalizedLoopAge)
    )
    Mesh Scale Mode=::NewEnumerator4
    UsePositionOffset=false
    Write Material Random=true
   )
   &CameraOffset(Camera Offset Amount=40)
   &ShapeLocation(
    MaskSpawnGroup=false
    Override Seed=false
    Sphere Radius=&Multiply_Float(A=1000, B=$NPC.NPC_DragonCart_Orb_Rift.RiftScale)
    UseOwnerScale=true
   )
   /SetVariables(Particles.MeshOrientation=&self:MakeQuatFromXandZ_02(
     XAxis:Vector3f=&Multiply_VectorByFloat(Vector=$Particles.ShapeLocation.ShapeNormal, Float=-1)
     ZAxis:Vector3f=$Particles.ShapeLocation.ShapeBitangent
    )) {isSetParameters=true}
  ]
  particleUpdateScript [
   &ParticleState()
   &ScaleMeshSize(Scale Factor=&VectorFromFloat(Value=&FloatFromCurve(
      FloatCurve=DataInterface() {propertyValues {Curve {keys [
          {arriveTangent=6.1738, interpMode=RCIM_Cubic, leaveTangent=6.1738, tangentMode=RCTM_User, time=0, value=0.2}
          {arriveTangent=0.8, interpMode=RCIM_Cubic, leaveTangent=0.8, time=0.2, value=0.9}
          <at 1 1>
         ]}}}
      Scale Curve=2
     )))
   &ScaleColor(
    Scale Alpha=&FloatFromCurve(
     FloatCurve=DataInterface() {propertyValues {Curve {keys [
         {arriveTangent=0.055, interpMode=RCIM_Cubic, leaveTangent=0.055, tangentMode=RCTM_User, time=0, value=0}
         {arriveTangent=0.0021, interpMode=RCIM_Cubic, leaveTangent=0.0021, tangentMode=RCTM_User, time=0.75, value=1}
         <at 1 0 RCIM_Cubic>
        ]}}}
     Scale Curve=&FloatFromCurve(
      FloatCurve=DataInterface() {propertyValues {Curve {defaultValue=0, keys [{tangentMode=RCTM_User, time=0, value=0}, {tangentMode=RCTM_User, time=1, value=1.5}]}}}
      CurveIndex=$Emitter.NormalizedLoopAge
     )
    )
    ScaleA=true
    ScaleRGB=true
   )
   &DynamicMaterialParameters(
    Index 0 Param 1=&FloatFromCurve(FloatCurve=DataInterface() {propertyValues {Curve {
        defaultValue=1
        keys [
         {arriveTangent=4.6205, interpMode=RCIM_Cubic, leaveTangent=4.6205, tangentMode=RCTM_User, time=0, value=0}
         {arriveTangent=1, interpMode=RCIM_Cubic, leaveTangent=1, time=0.2046, value=0.6309}
         {interpMode=RCIM_Cubic, leaveTangent=2.6157, time=1, value=1}
        ]
       }}})
    Index 0 Param 2=&Add_Float(
     A=&RandomRangeFloat(Minimum=0, Maximum=1)
     B=&FloatFromCurve(
      FloatCurve=DataInterface() {propertyValues {Curve {
         defaultValue=0
         keys [
          {arriveTangent=1.647, interpMode=RCIM_Cubic, leaveTangent=1.647, tangentMode=RCTM_User, time=0, value=0}
          {interpMode=RCIM_Cubic, tangentMode=RCTM_User, time=1, value=1}
         ]
        }}}
      Scale Curve=&RandomRangeFloat(Minimum=0.5, Maximum=1.5)
     )
    )
    Index 0 Param 3=&RandomRangeFloat(Minimum=1500, Maximum=2000)
    Index 0 Param 4=1
    Module.Index0Float0Write=true
    Module.Index0Float1Write=true
    Module.Index0Float2Write=true
    Module.Index0Float3Write=true
   )
  ]
  rendererClasses [&NiagaraMeshRendererProperties]
  renderers [{rendererClass=&NiagaraMeshRendererProperties, rendererIndex=0}]
 }
}
summary {
 systemName=NS_DCN_Rift_Buildup
 emitters [
  {emitterName=NE_Zaps, rendererClasses [&NiagaraMeshRendererProperties]}
  {bEnabled=false, emitterName=ElectricArcs, rendererClasses [&NiagaraSpriteRendererProperties]}
  {bEnabled=false, emitterName=Splines001, rendererClasses [&NiagaraMeshRendererProperties]}
  {
   emitterName=NE_Electricity_Wind
   rendererClasses [&NiagaraSpriteRendererProperties, &NiagaraMeshRendererProperties]
  }
  {bEnabled=false, emitterName=Minimal, rendererClasses [&NiagaraMeshRendererProperties]}
 ]
 userVariables [
  User.BuildupDuration {type=NiagaraFloat, defaultValue=30, description=""}
  User.SizeNormalized {type=NiagaraFloat, defaultValue=6, description=""}
  User.ZapAudioVolume {type=NiagaraFloat, defaultValue=0, description=""}
 ]
}
