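# Niagara system for the back layer of the DragonCart rift buildup, intensifying over time.
# CrackArms bursts partitioned crack meshes with per-partition orientation, mesh and scale.
# Minimal001 merges the disabled FlareLight, Horizontal_Shockwave and Minimal emitters, using spawn-group switches for flare, darkening, shockwaves and RiftScale-sized lightning.
# Minimal001's loop follows User.BuildupDuration, curves ramp spawning, size and brightness, and alpha scales by NearFade and time of day.
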NiagaraSystem /DragonCartNarrative/Assets/VFX/RiftBuildUp/NS_DCN_Rift_Buildup_Back
dependencies {
 usedDataInterfaces [
  &NiagaraDataInterfaceCurve
  &NiagaraDataInterfaceArrayFloat
  &NiagaraDataInterfaceArrayFloat3
  &NiagaraDataInterfaceArrayInt32
 ]
 usedDynamicInputs []
 usedModules [
  &SystemState
  &self:SystemSpawnScript.SystemScriptSource.SystemScriptGraph.NiagaraNodeAssignment_0.SetVariables_#1
 ]
}
emitters {
 CrackArms {
  emitterUpdateScript [
   &EmitterState()
   &SpawnBurst_Instantaneous(Spawn Count=4, Use Spawn Probability=false, UseLoopCountLimit=false)
   &SpawnBurst_Instantaneous(Spawn Count=4, Spawn Group=1, Use Spawn Probability=false, UseLoopCountLimit=false) {module=SpawnBurst_Instantaneous001}
  ]
  particleSpawnScript [
   &PartitionParticles(Number Of Partitions=4)
   &InitializeParticle(
    Lifetime=&RandomRangeFloat(Minimum=10, Maximum=13)
    Mass Mode=::NewEnumerator1
    Mesh Scale=&VectorFromFloat(Value=&RandomRangeFloat(Minimum=0.5, Maximum=1))
    Mesh Scale Mode=::NewEnumerator4
    UsePositionOffset=false
    Write Material Random=true
   ) {module=InitializeParticle001}
   &InitialMeshOrientation(
    Mesh Axis To Orient=<0 0 1>
    Mesh Orientation Mode=::NewEnumerator2
    Orientation Vector=&RotateVectorByQuaternion(
     Vector To Rotate=<0 0 1>
     Quaternion=&MakeQuatFromAxisAngle(
      Axis=<1 0 0>
      AngleInDegrees=&SelectFloatFromArray(
       Direct Array Index=$Particles.PartitionParticles.Partition
       Float Selection Array=DataInterface() {propertyValues {FloatData [-25,140,240,-80]}}
      )
     )
    )
    Rotation=&MaskVectorBySpawnGroup(
     Masked Vector=&SelectVectorFromArray(
      Direct Array Index=$Particles.PartitionParticles.Partition
      Vector Selection Array=DataInterface() {propertyValues {FloatData [<0.1 0 0.48>, <-0.05 0 0.2>, <0 0 0.25>, <0 0 -0.25>]}}
     )
     Passthrough Vector=&SelectVectorFromArray(
      Direct Array Index=$Particles.PartitionParticles.Partition
      Vector Selection Array=DataInterface() {propertyValues {FloatData [<-0.2 0.15 0.15>, <0.23 0.7 0.1>, <-0.4 0.25 0.6>, <0.4 0.2 0.3>]}}
     )
    )
    Use Rotation Vector=true
   )
   &CameraOffset(Camera Offset Amount=30)
   /SetVariables(Particles.MeshIndex=&SelectIntFromArray(Direct Array Index=$Particles.PartitionParticles.Partition, Int Selection Array=DataInterface() {propertyValues {IntData [0,1,0,1]}})) {isSetParameters=true}
  ]
  particleUpdateScript [
   &ParticleState(Kill Particles When Lifetime Has Elapsed=false, Let Infinitely Lived Particles Die When Emitter Deactivates=true)
   &ScaleColor(
    Scale Alpha=&Multiply_Float(A=$System.NearFade, B=&FNTimeOfDayScalar(Night Scale=0.5, Morning Scale=0.8, Evening Scale=0.6))
    Scale RGB=&VectorFromFloat(Value=&FloatFromCurve(FloatCurve=DataInterface() {propertyValues {Curve {keys [<at 0 1>,<at 1 2>]}}}, CurveIndex=$Emitter.NormalizedLoopAge))
    ScaleA=true
    ScaleRGB=true
   )
   &ScaleMeshSize(Scale Factor=&VectorFromFloat(Value=&Lerp_Float(
      A=&SelectFloatFromArray(
       Direct Array Index=$Particles.PartitionParticles.Partition
       Float Selection Array=DataInterface() {propertyValues {FloatData [0.45,0.3,0.5,0.4]}}
      )
      B=&SelectFloatFromArray(
       Direct Array Index=$Particles.PartitionParticles.Partition
       Float Selection Array=DataInterface() {propertyValues {FloatData [0.67,0.8,0.7,0.74]}}
      )
      Alpha=0.5
     )))
   &ScaleMeshSize(Scale Factor=&VectorFromFloat(Value=&RandomRangeFloat(Minimum=5, Maximum=7))) {module=ScaleMeshSize001}
   &DynamicMaterialParameters(
    Index 0 Param 1=&FloatFromCurve(
     FloatCurve=DataInterface() {propertyValues {Curve {
        defaultValue=0
        keys [
         {
          arriveTangent=-0.0603
          arriveTangentWeight=0.6451
          interpMode=RCIM_Cubic
          leaveTangent=-0.0603
          leaveTangentWeight=0.6451
          tangentMode=RCTM_User
          tangentWeightMode=RCTWM_WeightedBoth
          time=0
          value=0
         }
         {arriveTangentWeight=0.4549, interpMode=RCIM_Cubic, leaveTangentWeight=0.0165, tangentMode=RCTM_User, time=0.98, value=0.9973}
         {arriveTangent=-0.039, interpMode=RCIM_Cubic, leaveTangent=-0.039, tangentMode=RCTM_User, time=1, value=0}
        ]
       }}}
     CurveIndex=$Emitter.NormalizedLoopAge
    )
    Index 0 Param 2=1
    Index 0 Param 3=&FloatFromCurve(
     FloatCurve=DataInterface() {propertyValues {Curve {defaultValue=0, keys [{tangentMode=RCTM_User, time=0, value=0}, {tangentMode=RCTM_User, time=1, value=5}]}}}
     CurveIndex=$Emitter.NormalizedLoopAge
     Scale Curve=2
    )
    Module.Index0Param0Write=true
    Module.Index0Param1Write=true
    Module.Index0Param2Write=true
    Module.Index0Param3Write=false
   )
  ]
  rendererClasses [&NiagaraMeshRendererProperties]
  renderers [{rendererClass=&NiagaraMeshRendererProperties, rendererIndex=0}]
 }
 FlareLight {
  bEnabled=false
  emitterUpdateScript [
   &EmitterState()
   &SpawnBurst_Instantaneous(Spawn Count=2, Spawn Time=0.1, Use Spawn Probability=false, UseLoopCountLimit=false)
  ]
  particleSpawnScript [&InitializeParticle(Lifetime=28, Sprite Size=<300 200>, Sprite Size Mode=::NewEnumerator4, UsePositionOffset=false, Write Material Random=true)]
  particleUpdateScript [
   &ParticleState()
   &ScaleSpriteSize(
    Uniform Curve Scale=6.5
    Uniform Curve Sprite Scale=DataInterface() {propertyValues {Curve {keys [
        {arriveTangent=6.7103, interpMode=RCIM_Cubic, leaveTangent=6.7103, tangentMode=RCTM_User, time=0.0022, value=2}
        {arriveTangent=-0.5471, interpMode=RCIM_Cubic, leaveTangent=-0.5471, tangentMode=RCTM_User, time=0.8913, value=4}
       ]}}}
   )
   &CameraOffset(Camera Offset Amount=-200)
   /SetVariables(Particles.DarkenColor:LinearColor=&MakeLinearColorFromVectorAndFloat(
     "Vector (RGB)"=&MakeVectorFromLinearColorRGB(Color=<rgba 0.02 0.001 0 1>)
     "Float (Alpha)"=&FloatFromCurve(
      FloatCurve=DataInterface() {propertyValues {Curve {
         defaultValue=0
         keys [
          {arriveTangent=3.5605, interpMode=RCIM_Cubic, leaveTangent=3.5605, tangentMode=RCTM_User, time=0, value=0}
          {arriveTangent=0.2783, interpMode=RCIM_Cubic, leaveTangent=0.2783, tangentMode=RCTM_User, time=0.3448, value=0.4511}
          {arriveTangent=-0.2888, interpMode=RCIM_Cubic, leaveTangent=-0.2888, tangentMode=RCTM_User, time=0.7133, value=0.4755}
          {arriveTangent=-5.224, interpMode=RCIM_Cubic, leaveTangent=-5.224, tangentMode=RCTM_User, time=1, value=0}
         ]
        }}}
      Scale Curve=&FNTimeOfDayScalar(Night Scale=0.4, Morning Scale=0.6, Day Scale=0.7, Evening Scale=0.5)
     )
    )) {isSetParameters=true}
   /SetVariables(Particles.DarkenSize:Vector2f=&Multiply_Vector2D_ByFloat(Float=1.2, Vector2D=$Particles.SpriteSize)) {isSetParameters=true}
   &DynamicMaterialParameters(
    Index 0 Param 1=1
    Index 0 Param 2=1
    Index 0 Param 3=&FloatFromCurve(FloatCurve=DataInterface() {propertyValues {Curve {
        defaultValue=1
        keys [
         {tangentMode=RCTM_User, time=0, value=1}
         {arriveTangent=-0.0405, interpMode=RCIM_Cubic, leaveTangent=-0.0405, tangentMode=RCTM_User, time=0.7009, value=1}
         {arriveTangent=-8.0089, interpMode=RCIM_Cubic, leaveTangent=-8.0089, tangentMode=RCTM_User, time=1, value=0}
        ]
       }}})
    Index 3 Param 1=2
    Index 3 Param 2=5
    Module.Index0Float0Write=true
    Module.Index0Float1Write=true
    Module.Index0Float2Write=true
    Module.Index0Float3Write=false
    Module.Index3Float0Write=true
    Module.Index3Float1Write=true
    Module.Index3Float2Write=false
    Module.Index3Float3Write=false
    Write Parameter Index 3=::NewEnumerator1
   )
   &ScaleColor(
    Scale Alpha=&FloatFromCurve(FloatCurve=DataInterface() {propertyValues {Curve {
        defaultValue=0
        keys [
         {arriveTangent=1.9805, interpMode=RCIM_Cubic, leaveTangent=1.9805, tangentMode=RCTM_User, time=0, value=0}
         {arriveTangent=0.4913, interpMode=RCIM_Cubic, leaveTangent=0.4913, tangentMode=RCTM_User, time=1, value=1}
        ]
       }}})
    ScaleA=true
    ScaleRGB=false
   )
   &ScaleColor(
    Scale Alpha=&FNTimeOfDayScalar(Night Scale=0.5, Morning Scale=0.8, Evening Scale=0.6)
    ScaleA=true
    ScaleRGB=true
   ) {module=ScaleColor001}
  ]
  rendererClasses [&NiagaraSpriteRendererProperties]
  renderers [
   {rendererClass=&NiagaraSpriteRendererProperties}
   {rendererClass=&NiagaraSpriteRendererProperties, rendererIndex=1}
  ]
 }
 Horizontal_Shockwave {
  bEnabled=false
  emitterUpdateScript [
   &EmitterState()
   &SpawnRate(
    SpawnRate=&FloatFromCurve(
     FloatCurve=DataInterface() {propertyValues {Curve {
        defaultValue=0
        keys [
         {arriveTangent=-0.2087, interpMode=RCIM_Cubic, leaveTangent=-0.2087, tangentMode=RCTM_User, time=0, value=0}
         <at 0.695 2 RCIM_Cubic>
         {arriveTangent=-0.6522, interpMode=RCIM_Cubic, leaveTangent=-0.6522, tangentMode=RCTM_User, time=1, value=0}
        ]
       }}}
     CurveIndex=$Emitter.NormalizedLoopAge
    )
    Use Spawn Probability=false
   )
   &SpawnBurst_Instantaneous(Spawn Time=2, Use Spawn Probability=false, UseLoopCountLimit=false) {enabled=false, module=SpawnBurst_Instantaneous002}
  ]
  particleSpawnScript [
   &InitializeParticle(
    Color=<rgba 3 0.6148 0 1>
    Lifetime=&FloatFromCurve(
     FloatCurve=DataInterface() {propertyValues {Curve {keys [<at 0 2 RCIM_Cubic>, {arriveTangent=-2.1989, tangentMode=RCTM_User, time=1, value=1}]}}}
     CurveIndex=$Emitter.NormalizedLoopAge
    )
    Mesh Scale Mode=::NewEnumerator3
    Mesh Uniform Scale=&Multiply_Float(
     A=120
     B=&FloatFromCurve(
      FloatCurve=DataInterface() {propertyValues {Curve {
         defaultValue=1
         keys [
          {arriveTangent=1.8352, interpMode=RCIM_Cubic, leaveTangent=1.8352, tangentMode=RCTM_User, time=0, value=0.3106}
          {arriveTangent=0.7366, interpMode=RCIM_Cubic, leaveTangent=0.7366, tangentMode=RCTM_User, time=0.5018, value=0.8364}
          {arriveTangent=0.0224, interpMode=RCIM_Cubic, leaveTangent=0.0224, tangentMode=RCTM_User, time=1, value=1}
         ]
        }}}
      CurveIndex=$Emitter.NormalizedLoopAge
     )
    )
    Position Offset=<0 0 100>
    UsePositionOffset=true
    Write Material Random=true
   )
   &CameraOffset(Camera Offset Amount=100)
   &InitialMeshOrientation(Mesh Orientation Mode=::NewEnumerator2, Use Rotation Vector=true)
  ]
  particleUpdateScript [
   &ParticleState()
   &ScaleMeshSize(Scale Factor=&VectorFromFloat(Value=&FloatFromCurve(
      FloatCurve=DataInterface() {propertyValues {Curve {
         defaultValue=1
         keys [
          {arriveTangent=0.0764, interpMode=RCIM_Cubic, leaveTangent=0.0764, tangentMode=RCTM_User, time=0, value=0.2}
          {arriveTangent=0.0868, interpMode=RCIM_Cubic, leaveTangent=0.0868, tangentMode=RCTM_User, time=1, value=1}
         ]
        }}}
      Scale Curve=&Multiply_Float(A=0.6, B=2)
     )))
   &DynamicMaterialParameters(
    Index 0 Param 2=&FloatFromCurve(FloatCurve=DataInterface() {propertyValues {Curve {defaultValue=1, keys [{tangentMode=RCTM_User, time=0, value=0}, {tangentMode=RCTM_User, time=1, value=1}]}}})
    Module.Param0WriteEnabled=true
    Module.Param1WriteEnabled=true
    Module.Param2WriteEnabled=true
    Module.Param3WriteEnabled=true
   )
   &ScaleColor(
    Scale Alpha=&FloatFromCurve(FloatCurve=DataInterface() {propertyValues {Curve {
        defaultValue=0
        keys [
         {arriveTangent=1.217, interpMode=RCIM_Cubic, leaveTangent=1.217, tangentMode=RCTM_User, time=0, value=0}
         {arriveTangent=0.3245, interpMode=RCIM_Cubic, leaveTangent=0.3245, tangentMode=RCTM_User, time=0.5231, value=0.351}
         {arriveTangent=0.0199, interpMode=RCIM_Cubic, leaveTangent=0.0199, tangentMode=RCTM_User, time=1, value=0}
        ]
       }}})
    ScaleA=true
    ScaleRGB=true
   )
   &ScaleColor(
    Scale Alpha=&FNTimeOfDayScalar(Night Scale=0.5, Morning Scale=0.8, Evening Scale=0.6)
    ScaleA=true
    ScaleRGB=true
   ) {module=ScaleColor001}
  ]
  rendererClasses [&NiagaraMeshRendererProperties]
  renderers [{rendererClass=&NiagaraMeshRendererProperties, rendererIndex=0}]
 }
 Minimal {
  bEnabled=false
  emitterUpdateScript [
   &EmitterState()
   &SpawnRate(Spawn Probability=0.7, SpawnRate=20, Use Spawn Probability=true)
  ]
  particleSpawnScript [
   &InitializeParticle(
    Color=<rgba 5 2.6183 0.1123 1>
    Lifetime=&FloatFromCurve(
     FloatCurve=DataInterface() {propertyValues {Curve {
        defaultValue=1
        keys [
         {interpMode=RCIM_Cubic, tangentMode=RCTM_User, time=0, value=1}
         {arriveTangent=-1.6186, interpMode=RCIM_Cubic, leaveTangent=-1.6186, tangentMode=RCTM_User, time=1, value=0.5}
        ]
       }}}
     CurveIndex=$Emitter.NormalizedLoopAge
     Scale Curve=&RandomRangeFloat(Minimum=0.3, Maximum=0.5)
    )
    Mesh Scale Max=<30 25 25>
    Mesh Scale Min=<20 10 10>
    Mesh Scale Mode=::NewEnumerator2
    UsePositionOffset=false
    Write Material Random=true
   )
   &ShapeLocation(
    Cylinder Height=50
    Cylinder Radius=&Multiply_Float(A=1100, B=$NPC.NPC_DragonCart_Orb_Rift.RiftScale)
    MaskSpawnGroup=false
    Override Seed=false
    Shape Primitive=::NewEnumerator1
    Surface Only=true
    Use Endcaps In Surface Only Mode=false
    UseOwnerScale=true
   )
   &InitialMeshOrientation(
    Mesh Orientation Mode=::NewEnumerator2
    Orientation Vector=$Particles.ShapeLocation.ShapeNormal
    Rotation=&RandomRangeVector(Minimum=<0 0 0>, Maximum=<1 0 0>)
    Use Rotation Vector=true
   )
   &CameraOffset(Camera Offset Amount=-1500)
  ]
  particleUpdateScript [
   &ParticleState()
   &ScaleMeshSize(Scale Factor=&VectorFromFloat(Value=&FloatFromCurve(
      FloatCurve=DataInterface() {propertyValues {Curve {
         defaultValue=1
         keys [{arriveTangent=1.1736, interpMode=RCIM_Cubic, leaveTangent=1.1736, tangentMode=RCTM_User, time=0, value=0.4447}, <at 1 1 RCIM_Cubic>]
        }}}
      CurveIndex=$Emitter.NormalizedLoopAge
     )))
   &DynamicMaterialParameters(
    Index 0 Param 1=&FloatFromCurve(FloatCurve=DataInterface() {propertyValues {Curve {keys [
         <at 0 1 RCIM_Cubic>
         {arriveTangent=-1.2958, interpMode=RCIM_Cubic, leaveTangent=-1.2958, tangentMode=RCTM_User, time=0.4949, value=0.5903}
         {arriveTangent=-1.8899, interpMode=RCIM_Cubic, leaveTangent=-1.8899, tangentMode=RCTM_User, time=0.5, value=0.4097}
         <at 1 0>
        ]}}})
    Index 0 Param 2=0.5
    Index 0 Param 3=&FloatFromCurve(FloatCurve=DataInterface() {propertyValues {Curve {keys [<at 0 0.17>,<at 0.6318 0.3857>,<at 0.9163 1.2358>,<at 1 0.1>]}}})
    Index 0 Param 4=&RandomRangeFloat(Minimum=0, Maximum=1, Evaluation Type=::NewEnumerator1)
    Index 1 Param 1=2
    Index 1 Param 2=2
    Index 1 Param 3=1
    Module.Index0Float0Write=true
    Module.Index0Float1Write=true
    Module.Index0Float2Write=true
    Module.Index0Float3Write=true
    Module.Index1Float0Write=true
    Module.Index1Float1Write=true
    Module.Index1Float2Write=true
    Module.Index1Float3Write=true
    Write Parameter Index 1=::NewEnumerator1
   )
   &ScaleColor(
    Scale Alpha=&FloatFromCurve(FloatCurve=DataInterface() {propertyValues {Curve {
        defaultValue=1
        keys [{arriveTangent=3.0787, interpMode=RCIM_Cubic, time=0, value=0}, {interpMode=RCIM_Cubic, tangentMode=RCTM_User, time=0.75, value=1}]
       }}})
    ScaleA=true
    ScaleRGB=false
   )
  ]
  rendererClasses [&NiagaraMeshRendererProperties]
  renderers [{rendererClass=&NiagaraMeshRendererProperties, rendererIndex=0}]
 }
 Minimal001 {
  emitterUpdateScript [
   &EmitterState(
    Life Cycle Mode=::NewEnumerator1
    Loop Behavior=::NewEnumerator1
    Loop Duration=&Add_Float(A=$User.BuildupDuration, B=1)
    UseLoopDelay=false
   )
   &SpawnBurst_Instantaneous(Spawn Time=0.1, Use Spawn Probability=false, UseLoopCountLimit=false) {module=SpawnBurst_Instantaneous001}
   &SpawnBurst_Instantaneous(Spawn Group=1, Spawn Time=0.1, Use Spawn Probability=false, UseLoopCountLimit=false) {module=SpawnBurst_Instantaneous002}
   &SpawnRate(
    SpawnGroup=2
    SpawnRate=&FloatFromCurve(
     FloatCurve=DataInterface() {propertyValues {Curve {
        defaultValue=0
        keys [
         {arriveTangent=0.4523, interpMode=RCIM_Cubic, leaveTangent=0.4523, tangentMode=RCTM_User, time=0, value=0}
         {arriveTangent=9.0713, interpMode=RCIM_Constant, tangentMode=RCTM_User, time=0.95, value=3}
         {arriveTangent=10.754, interpMode=RCIM_Constant, time=0.97, value=0}
        ]
       }}}
     CurveIndex=$Emitter.NormalizedLoopAge
     Scale Curve=1.2
    )
    Use Spawn Probability=false
   ) {module=SpawnRate001}
   &SpawnRate(
    Spawn Probability=0.7
    SpawnGroup=3
    SpawnRate=&FloatFromCurve(
     FloatCurve=DataInterface() {propertyValues {Curve {
        defaultValue=0
        keys [
         {arriveTangent=0.4523, interpMode=RCIM_Cubic, leaveTangent=0.4523, tangentMode=RCTM_User, time=0, value=0}
         {
          arriveTangent=0.4721
          arriveTangentWeight=0.2554
          interpMode=RCIM_Cubic
          leaveTangent=0.4721
          leaveTangentWeight=0.2554
          tangentMode=RCTM_User
          tangentWeightMode=RCTWM_WeightedBoth
          time=0.5
          value=0
         }
         {arriveTangent=61.3748, interpMode=RCIM_Constant, tangentMode=RCTM_User, time=0.95, value=15}
         {arriveTangent=10.754, interpMode=RCIM_Constant, leaveTangent=10.754, time=0.97, value=0}
        ]
       }}}
     CurveIndex=$Emitter.NormalizedLoopAge
    )
    Use Spawn Probability=true
   )
   &SpawnBurst_Instantaneous(Spawn Group=2, Spawn Time=28.5, Use Spawn Probability=false, UseLoopCountLimit=false) {enabled=false}
   &SpawnBurst_Instantaneous(Spawn Count=3, Spawn Group=3, Spawn Time=28.5, Use Spawn Probability=true, UseLoopCountLimit=false) {enabled=false, module=SpawnBurst_Instantaneous003}
  ]
  particleSpawnScript [
   /SetVariables(Particles.VisibilityTag=&SpawnGroup(Max Spawn Group=4)) {isSetParameters=true}
   &InitializeParticle(
    Color=&MergeHelper_ColorSwitch(
     0=<rgba 1 1 1 1>
     1=<rgba 0.02 0.001 0 1>
     2=<rgba 2 0.4098 0 0.5>
     3=<rgba 5 2.6183 0.1123 1>
     Index=&SpawnGroup()
     Num=::NewEnumerator2
    )
    Lifetime=&MergeHelper_FloatSwitch(
     0=31
     1=31
     2=&FloatFromCurve(
      FloatCurve=DataInterface() {propertyValues {Curve {keys [
          {arriveTangent=-1.3546, interpMode=RCIM_Cubic, leaveTangent=-1.3546, tangentMode=RCTM_User, time=0, value=1.5}
          {arriveTangent=0.287, tangentMode=RCTM_User, time=1, value=1}
         ]}}}
      CurveIndex=$Emitter.NormalizedLoopAge
     )
     3=&RandomRangeFloat(Minimum=0.2, Maximum=0.3)
     4=&RandomRangeFloat(Minimum=10, Maximum=13)
     Index=&SpawnGroup()
     Num=::NewEnumerator2
    )
    Mesh Scale=&MergeHelper_VectorSwitch(
     0=<0 0 0>
     1=<0 0 0>
     2=&VectorFromFloat(Value=&Multiply_Float(A=120, B=1))
     3=&RandomRangeVector(Minimum=<20 10 10>, Maximum=<30 25 25>)
     4=&VectorFromFloat(Value=&RandomRangeFloat(Minimum=0.5, Maximum=1))
     Index=&SpawnGroup()
     Num=::NewEnumerator2
    )
    Mesh Scale Mode=::NewEnumerator4
    Sprite Size=&MergeHelper_Vector2Switch(
     0=<300 200>
     1=&Multiply_Vector2D_ByFloat(Float=1.2, Vector2D=<300 200>)
     2=<0 0>
     3=<0 0>
     4=?
     Index=&SpawnGroup()
     Num=::NewEnumerator2
    )
    Sprite Size Mode=::NewEnumerator4
    UsePositionOffset=false
    Write Material Random=true
   )
   &CameraOffset(Camera Offset Amount=&MergeHelper_FloatSwitch(0=-200, 1=-200, 2=100, 3=-500, Index=&SpawnGroup(), Num=::NewEnumerator2))
   &ShapeLocation(
    Cylinder Height=&MergeHelper_FloatSwitch(0=0, 1=0, 2=0, 3=0, Index=&SpawnGroup(), Num=::NewEnumerator2)
    Cylinder Radius=&MergeHelper_FloatSwitch(
     0=0
     1=0
     2=0
     3=&Multiply_Float(A=1100, B=$NPC.NPC_DragonCart_Orb_Rift.RiftScale)
     Index=&SpawnGroup()
     Num=::NewEnumerator2
    )
    MaskSpawnGroup=false
    Override Seed=false
    Shape Primitive=::NewEnumerator1
    Surface Only=true
    Use Endcaps In Surface Only Mode=false
    UseOwnerScale=true
   )
   &InitialMeshOrientation(
    Mesh Axis To Orient=&MergeHelper_VectorSwitch(0=<0 0 0>, 1=<0 0 0>, 2=<1 0 0>, 3=<1 0 0>, Index=&SpawnGroup(), Num=::NewEnumerator2)
    Mesh Orientation Mode=::NewEnumerator2
    Orientation Vector=&MergeHelper_VectorSwitch(
     0=<0 0 0>
     1=<0 0 0>
     2=<1 0 0>
     3=$Particles.ShapeLocation.ShapeNormal
     Index=&SpawnGroup()
     Num=::NewEnumerator2
    )
    Rotation=&MergeHelper_VectorSwitch(
     0=<0 0 0>
     1=<0 0 0>
     2=<0 0 0>
     3=&RandomRangeVector(Minimum=<0 0 0>, Maximum=<1 0 0>)
     Index=&SpawnGroup()
     Num=::NewEnumerator2
    )
    Use Rotation Vector=true
   )
  ]
  particleUpdateScript [
   &ParticleState()
   &UpdateMeshOrientation(
    Rotation Rate=&MergeHelper_FloatSwitch(0=0, 1=0, 2=0, 3=0, Index=&SpawnGroup(), Num=::NewEnumerator2)
    Rotation Vector=&MergeHelper_VectorSwitch(
     0=<0 0 0>
     1=<0 0 0>
     2=<0 0 0>
     3=&RandomRangeVector(Minimum=<-1 -1 -1>, Maximum=<1 1 1>, Evaluation Type=::NewEnumerator1)
     Index=&SpawnGroup()
     Num=::NewEnumerator2
    )
   )
   &ScaleSpriteSize(
    Uniform Curve Scale=6.5
    Uniform Curve Sprite Scale=DataInterface() {propertyValues {Curve {keys [
        {arriveTangent=6.7103, interpMode=RCIM_Cubic, leaveTangent=6.7103, tangentMode=RCTM_User, time=0.0022, value=2}
        {arriveTangent=-0.5471, interpMode=RCIM_Cubic, leaveTangent=-0.5471, tangentMode=RCTM_User, time=0.8913, value=4}
        <at 1 0 RCIM_Cubic>
       ]}}}
   )
   &ScaleMeshSize(Scale Factor=&MergeHelper_VectorSwitch(
     0=<0 0 0>
     1=<0 0 0>
     2=&VectorFromFloat(Value=&FloatFromCurve(
       FloatCurve=DataInterface() {propertyValues {Curve {
          defaultValue=1
          keys [
           {arriveTangent=-0.0463, interpMode=RCIM_Cubic, leaveTangent=-0.0463, tangentMode=RCTM_User, time=0, value=0.4143}
           {arriveTangent=3.6576, interpMode=RCIM_Cubic, leaveTangent=3.6576, tangentMode=RCTM_User, time=1, value=1.5}
          ]
         }}}
       CurveIndex=$Emitter.NormalizedLoopAge
       Scale Curve=1.3
      ))
     3=&VectorFromFloat(Value=&FloatFromCurve(
       FloatCurve=DataInterface() {propertyValues {Curve {
          defaultValue=1
          keys [
           {arriveTangent=0.6394, interpMode=RCIM_Cubic, leaveTangent=0.6394, tangentMode=RCTM_User, time=0, value=0.5859}
           {arriveTangent=-0.0187, interpMode=RCIM_Cubic, leaveTangent=-0.0187, tangentMode=RCTM_User, time=1, value=1}
          ]
         }}}
       CurveIndex=$Emitter.NormalizedLoopAge
      ))
     Index=&SpawnGroup()
     Num=::NewEnumerator2
    ))
   &ScaleColor(
    Scale Alpha=&MergeHelper_FloatSwitch(
     0=&FloatFromCurve(FloatCurve=DataInterface() {propertyValues {Curve {
         defaultValue=0
         keys [
          {arriveTangent=1.9805, interpMode=RCIM_Cubic, leaveTangent=1.9805, tangentMode=RCTM_User, time=0, value=0}
          {arriveTangent=0.4913, interpMode=RCIM_Cubic, leaveTangent=0.4913, tangentMode=RCTM_User, time=1, value=1}
         ]
        }}})
     1=&FloatFromCurve(FloatCurve=DataInterface() {propertyValues {Curve {
         defaultValue=0
         keys [
          {arriveTangent=3.5605, interpMode=RCIM_Cubic, leaveTangent=3.5605, tangentMode=RCTM_User, time=0, value=0}
          {arriveTangent=0.2783, interpMode=RCIM_Cubic, leaveTangent=0.2783, tangentMode=RCTM_User, time=0.3448, value=0.4511}
          {arriveTangent=-0.2888, interpMode=RCIM_Cubic, leaveTangent=-0.2888, tangentMode=RCTM_User, time=0.7133, value=0.4755}
          {arriveTangent=-5.224, interpMode=RCIM_Cubic, leaveTangent=-5.224, tangentMode=RCTM_User, time=1, value=0}
         ]
        }}})
     2=&FloatFromCurve(FloatCurve=DataInterface() {propertyValues {Curve {
         defaultValue=0
         keys [
          {arriveTangent=-0.3122, interpMode=RCIM_Cubic, leaveTangent=-0.3122, tangentMode=RCTM_User, time=0, value=0.1}
          {arriveTangent=-0.0595, interpMode=RCIM_Cubic, leaveTangent=-0.0595, tangentMode=RCTM_User, time=1, value=0}
         ]
        }}})
     3=1
     Index=&SpawnGroup()
     Num=::NewEnumerator2
    )
    Scale RGB=&MergeHelper_VectorSwitch(
     0=<1 1 1>
     1=<1 1 1>
     2=&VectorFromFloat(Value=&FloatFromCurve(
       FloatCurve=DataInterface() {propertyValues {Curve {
          defaultValue=1
          keys [
           {interpMode=RCIM_Cubic, tangentMode=RCTM_User, time=0, value=1}
           {
            arriveTangent=21.8445
            arriveTangentWeight=2.2898
            interpMode=RCIM_Cubic
            leaveTangent=21.8445
            leaveTangentWeight=2.2898
            tangentMode=RCTM_User
            tangentWeightMode=RCTWM_WeightedBoth
            time=1
            value=3
           }
          ]
         }}}
       CurveIndex=$Emitter.NormalizedLoopAge
      ))
     3=&VectorFromFloat(Value=&FloatFromCurve(FloatCurve=DataInterface() {propertyValues {Curve {keys [<at 0 1>,<at 0.3075 1 RCIM_Constant>,<at 0.3789 0.1765 RCIM_Constant>,<at 0.4247 0.9529 RCIM_Constant>,<at 0.766 1>,<at 0.7941 0.3176>,<at 0.8043 1>,<at 1 1>]}}}))
     Index=&SpawnGroup()
     Num=::NewEnumerator2
    )
    ScaleA=true
    ScaleRGB=true
   )
   &ScaleColor(
    Scale Alpha=&Multiply_Float(A=$System.NearFade, B=&FNTimeOfDayScalar(Night Scale=0.5, Morning Scale=0.8, Evening Scale=0.6))
    ScaleA=true
    ScaleRGB=true
   ) {module=ScaleColor001}
   /SetVariables(
    Particles.DarkenDyn:Vector4f=&MakeVector4(X=2, Y=5, Z=0, W=0)
    Particles.FlareDyn:Vector4f=&MakeVector4(
     X=1
     Y=1
     Z=&FloatFromCurve(FloatCurve=DataInterface() {propertyValues {Curve {
         defaultValue=1
         keys [
          {tangentMode=RCTM_User, time=0, value=1}
          {arriveTangent=-0.0405, interpMode=RCIM_Cubic, leaveTangent=-0.0405, tangentMode=RCTM_User, time=0.7009, value=1}
          {arriveTangent=-8.0089, interpMode=RCIM_Cubic, leaveTangent=-8.0089, tangentMode=RCTM_User, time=1, value=0}
         ]
        }}})
     W=0
    )
    Particles.LightnningDyn:Vector4f=&MakeVector4(
     X=&FloatFromCurve(FloatCurve=DataInterface() {propertyValues {Curve {keys [
          <at 0 1 RCIM_Cubic>
          {arriveTangent=-1.2958, interpMode=RCIM_Cubic, leaveTangent=-1.2958, tangentMode=RCTM_User, time=0.49, value=0.6}
          {arriveTangent=-1.8899, interpMode=RCIM_Cubic, leaveTangent=-1.8899, tangentMode=RCTM_User, time=0.5, value=0.4}
          <at 1 0>
         ]}}})
     Y=&RandomRangeFloat(Minimum=0.35, Maximum=0.5)
     Z=&FloatFromCurve(FloatCurve=DataInterface() {propertyValues {Curve {keys [<at 0 0.17>,<at 0.6318 0.3857>,<at 0.9163 1.2358>,<at 1 0.1>]}}})
     W=&RandomRangeFloat(Minimum=0, Maximum=1, Evaluation Type=::NewEnumerator1)
    )
    Particles.LightnningDyn1:Vector4f=&MakeVector4(X=2, Y=2, Z=1, W=0)
    Particles.ShockwaveDyn:Vector4f=&MakeVector4(
     X=0
     Y=&FloatFromCurve(FloatCurve=DataInterface() {propertyValues {Curve {
         defaultValue=0
         keys [{interpMode=RCIM_Cubic, tangentMode=RCTM_User, time=0, value=0}, {interpMode=RCIM_Cubic, tangentMode=RCTM_User, time=1, value=2}]
        }}})
     Z=0
     W=0
    )
   ) {isSetParameters=true}
  ]
  rendererClasses [&NiagaraSpriteRendererProperties, &NiagaraMeshRendererProperties]
  renderers [
   {rendererClass=&NiagaraSpriteRendererProperties}
   {rendererClass=&NiagaraSpriteRendererProperties, rendererIndex=1}
   {rendererClass=&NiagaraMeshRendererProperties, rendererIndex=2}
   {rendererClass=&NiagaraMeshRendererProperties, rendererIndex=3}
  ]
 }
}
summary {
 systemName=NS_DCN_Rift_Buildup_Back
 emitters [
  {bEnabled=false, emitterName=FlareLight, rendererClasses [&NiagaraSpriteRendererProperties]}
  {bEnabled=false, emitterName=Horizontal_Shockwave, rendererClasses [&NiagaraMeshRendererProperties]}
  {bEnabled=false, emitterName=Minimal, rendererClasses [&NiagaraMeshRendererProperties]}
  {emitterName=CrackArms, rendererClasses [&NiagaraMeshRendererProperties]}
  {emitterName=Minimal001, rendererClasses [&NiagaraSpriteRendererProperties, &NiagaraMeshRendererProperties]}
 ]
 userVariables [
  User.BuildupDuration {type=NiagaraFloat, defaultValue=30, description=""}
  User.SizeNormalized {type=NiagaraFloat, defaultValue=1, description=""}
 ]
}
